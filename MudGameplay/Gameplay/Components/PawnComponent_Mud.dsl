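; Fortnite pawn component that tracks mud contact and covers pawns in mud from sliding, sprinting, landing or an attached parent.
; The server applies on-mud and covered GameplayEffects, gated by blocking tags and hotfix toggles, and removes cover on entering water.
; It caches default sliding parameters, swaps in mud values and rechecks them when slides end.
; A timer, updated when fire modifies cover duration, triggers the owning client's removal-warning cue.
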(event EventBeginPlay
  (bind _movementcomp (CallFunction|GetCharacterMovementComp))
  ;[1 1 1] BEGIN PLAY - Initialization
  (bind _asfort_pawn (Utilities|Casting|CastToFortPawn (Components|GetOwner)))
  (Variables|Default|SetPawn _asfort_pawn)
  ;;[1 1 1] Cache default sliding params at the very beginning, otherwise there is a chance we accidentally set the mud sliding params as the default if we go on and off mud very quickly
  (Utilities|IsValid _movementcomp
    (:Is%20Valid
      (Variables|Mud|Transient|SetDefaultSlidingControlParams (Variables|/Script/FortniteGame.FortMovementComp_Character|GetSlidingControlParams _movementcomp)))
    (:Is%20Not%20Valid
      _))
  (Networking|SwitchHasAuthority
    (:Authority
      ;;[1 1 1] Mud coverage from jump landing
      (EventDispatchers|BindEventtoOnPawnLanded (Variables|Default|GetPawn) (delegate Custom|OnPawnLanded))
      (bind _asfort_player_pawn (Utilities|Casting|CastToFortPlayerPawn (Variables|Default|GetPawn)))
      ;;[1 1 1] Enter water event
      (EventDispatchers|AssignOnPawnEnterWater _asfort_player_pawn (delegate Custom|HandlePawnEnterWater)))))

(event Custom|HandlePawnEnterWater
  (params (FortPlayerPawn :type Fort%20Player%20Pawn%20Object%20Reference))
  ;[1 1 1] BEGIN PLAY - Initialization
  ;;[1 1 1] Enter water event
  (Utilities|FlowControl|Sequence
    (:then_0
      ;;;[1 1 1] Check if we are already covered before trying to remove; for playing water+mud FX
      (if (Class|MudFunctionLibrary|IsCoveredinMud FortPlayerPawn)
        (CallFunction|OnEnterWaterWhileCovered)))
    (:then_1
      (CallFunction|EndMudCover))))

(event Custom|OnPawnLanded
  (params (Hit :type Hit%20Result%20Structure%20%28by%20ref%29))
  (bind (_bblockinghit _binitialoverlap _time _distance _location _impactpoint _normal _impactnormal _physmat) (Collision|BreakHitResult Hit))
  ;[1 1 1] BEGIN PLAY - Initialization
  ;;[1 1 1] Mud coverage from jump landing
  ;;;[1 1 1] Surface changes are not broadcasted until after the landed event occurs, so we need to check the surface on land instead of relying on IsInMud being set beforehand
  (if (Utilities|Enum|Equal(Enum) (Variables|/Script/PhysicsCore.PhysicalMaterial|GetSurfaceType _physmat) "SurfaceType30")
    (CallFunction|TryCoverinMud "NewEnumerator3")))

(event Custom|DebugRemoveMud
  ;[1 0.6676 0.179] DEBUG CHEATS
  (if (Pawn|IsLocallyControlled (Variables|Default|GetPawn))
    (CallFunction|RemoveMudServer)
    (CallFunction|EndMudCover)
    (elif (Networking|IsServer)
      (CallFunction|EndMudCover))))

(event Custom|RemoveMud_Server
  ;[1 0.6676 0.179] DEBUG CHEATS
  (CallFunction|EndMudCover))

(event Custom|DebugCoverMud)

(event EventEndPlay
  (params (EndPlayReason :type EEndPlayReason%20Enum :default "Destroyed"))
  ;[1 1 1] END PLAY - Clean Up
  (Networking|SwitchHasAuthority
    (:Authority
      (Utilities|IsValid (Variables|Default|GetPawn)))))

(event Custom|OnEnterWaterWhileCovered)

(event EventOnCurieFireTagCountChanged
  (params (Tag :type Gameplay%20Tag%20Structure) (NewCount :type Integer))
  ;[1 1 1] Fire (Curie) tag changes
  (if (> NewCount 0)
    (CallFunction|OnCurieFireTagAdded)))

(event Custom|OnCurieFireTagAdded
  ;[1 1 1] Fire (Curie) tag changes
  (bind _pawn (Variables|Default|GetPawn)
    (:then
      (bind _returnvalue (GameplayEffects|ApplyGameplayEffectSpecToSelf (Ability|GetAbilitySystemComponent _pawn) (GameplayEffects|MakeOutgoingSpec :self (Ability|GetAbilitySystemComponent _pawn) :GameplayEffectClass (Variables|Mud|GetGE_ModifyDurationFireApplied) :Level 0.0)))
      (Variables|Mud|SetModifyDurationFireAppliedHandle _returnvalue)
      (Development|PrintString "Fire tag")
      ;;[1 1 1] Fire changes the duration of the mud covered GE, so make sure we update the warning timer
      (CallFunction|UpdateRemovalWarningTimer))
    (:else)))

(event Custom|OwningClient_StartRemovalWarningCue
  (CallFunction|StartRemovalWarningCue))

(event Custom|OwningClient_TryEndRemovalWarningCue
  (CallFunction|TryEndRemovalWarningCue))

(event EventOnEnteredMud
  ;[1 1 1] Mud entering & exiting
  (Development|PrintString (Utilities|String|Append (Utilities|GetDisplayName (Variables|Default|GetPawn)) " entered mud (component implementation)") true true "(R=1.000000,G=0.400180,B=0.000000,A=1.000000)")
  (Variables|Mud|Transient|SetbIsInMud true)
  (Utilities|FlowControl|Sequence
    (:then_0
      (Utilities|SwitchHasAuthority
        (:Authority
          (CallFunction|ApplyonMudGameplayEffects))))
    (:then_1
      (if (Variables|Mud|Transient|GetbIsSliding)
        (CallFunction|OnStartedSlidinginMud)))
    (:then_2
      (if (Variables|Mud|Transient|GetbIsSprinting)
        (CallFunction|TryCoverinMud "NewEnumerator2")))))

(event EventOnExitedMud
  ;[1 1 1] Mud entering & exiting
  (Development|PrintString "Exited mud (component implementation)" true true "(R=1.000000,G=0.400180,B=0.000000,A=1.000000)")
  (Variables|Mud|Transient|SetbIsInMud false)
  (Utilities|SwitchHasAuthority
    (:Authority
      (CallFunction|RemoveonMudGameplayEffects))))

(event EventOnSlidingTagCountChanged
  (params (Tag :type Gameplay%20Tag%20Structure) (NewCount :type Integer))
  ;[1 1 1] Sliding in mud
  (if (> NewCount 0)
    (CallFunction|OnSlidingBegin)
    (else
      (CallFunction|OnSlidingEnd))))

(event Custom|OnSlidingBegin
  ;[1 1 1] Sliding in mud
  (Variables|Mud|Transient|SetbIsSliding true)
  (if (Variables|Mud|Transient|GetbIsInMud)
    (CallFunction|OnStartedSlidinginMud)))

(event Custom|OnSlidingEnd
  ;[1 1 1] Sliding in mud
  (Variables|Mud|Transient|SetbIsSliding false)
  ;;[1 1 1 1 12] This is a failsafe for mud sliding overrides not being removed properly by the ability, and a solve for a bug in which the params aren't reset properly if you're still sliding
  (CallFunction|CheckforResetSlidingParams)
  ;;[1 1 1 1 12] Similarly, application of sliding params may fail if we are already sliding, so try to apply them before the next slide
  (CallFunction|CheckforApplySlidingParams))

(event EventOnAttachedParentCoveredInMud
  (params (Payload :type Gameplay%20Event%20Data%20Structure%20%28by%20ref%29))
  ;[1 1 1] Attached parent covered in mud
  (CallFunction|TryCoverinMud "NewEnumerator0"))

(event EventOnSprintingTagCountChanged
  (params (Tag :type Gameplay%20Tag%20Structure) (NewCount :type Integer))
  ;[1 1 1] Sprinting in mud
  (if (> NewCount 0)
    (Variables|Mud|Transient|SetbIsSprinting true)
    (if (Variables|Mud|Transient|GetbIsInMud)
      (CallFunction|TryCoverinMud "NewEnumerator2"))
    (else
      (Variables|Mud|Transient|SetbIsSprinting false))))

(fn ApplyMudSlidingParameters
  (bind _movementcomp (CallFunction|GetCharacterMovementComp))
  (if (CallFunction|IsSlidingModifierEnabled)
    ;[1 1 1] This is a little odd but it looks like there's a chance sliding param changes don't stick if you're already sliding, but on the off-chance they do, we are attempting to set them in either case. However, when we are already sliding we leave AppliedSlidingParams false so we know to try and apply them again the next time we stop sliding. Gating with the bool so we don't try and apply every single time we stop sliding in mud just in case that has bad side effects/causes movement corrections, etc.
    (if (not (Variables|Mud|Transient|GetbIsSliding))
      (Variables|Default|SetAppliedSlidingParams true))
    (bind _structout (Utilities|Struct|SetmembersinSlidingControlParams (Utilities|Copy (Variables|/Script/FortniteGame.FortMovementComp_Character|GetSlidingControlParams _movementcomp)) (Variables|Mud|Sliding|GetMin%20Speed%20To%20Start%20Sliding) (Variables|Mud|Sliding|GetUphill%20Slide%20Start%20Boost) (Variables|Mud|Sliding|GetLevel%20Slide%20Start%20Boost) (Variables|Mud|Sliding|GetDownhill%20Slide%20Start%20Boost) (Variables|Mud|Sliding|GetBoost%20Level%20Ground%20Friction%20Multiplier) (Variables|Mud|Sliding|GetUphill%20Ground%20Friction%20Multiplier) (Variables|Mud|Sliding|GetLevel%20Ground%20Friction%20Multiplier) (Variables|Mud|Sliding|GetDownhill%20Ground%20Friction%20Multiplier) (Variables|Mud|Sliding|GetMax%20Forward%20Sliding%20Speed) (Variables|Mud|Sliding|GetMin%20Forward%20Sliding%20Speed)))
    (Variables|/Script/FortniteGame.FortMovementComp_Character|SetSlidingControlParams _structout _movementcomp)))

(fn ApplyOnMudGameplayEffects
  (bind _pawn (Variables|Default|GetPawn))
  (bind _returnvalue (Utilities|GetDisplayName _pawn))
  (bind _pawn_2 (Variables|Default|GetPawn)
    (:then
      (bind _output_get (LocalVariables|SetOwningASC (Ability|GetAbilitySystemComponent _pawn_2)))
      (Utilities|IsValid _output_get
        (:Is%20Valid
          (Utilities|FlowControl|Sequence
            (:then_0
              (Development|PrintString (Utilities|String|Append _returnvalue ": adding OnMud gameplay effects"))
              (for _array_element (Variables|Mud|GetOnMudGameplayEffects)
                (GameplayEffects|ApplyGameplayEffectToSelf :self (LocalVariables|GetOwningASC) :GameplayEffectClass _array_element :Level 0.0)))
            (:then_1
              (Development|PrintString (Utilities|String|Append _returnvalue ": adding Shared gameplay effects"))
              ;[1 1 1] Also try to apply shared mud effects. This should be safe if the player already has these from being covered in mud, they are each limited to a single stack per target
              (for _array_element_2 (Variables|Mud|GetSharedMudGameplayEffects)
                (GameplayEffects|ApplyGameplayEffectToSelf :self (LocalVariables|GetOwningASC) :GameplayEffectClass _array_element_2 :Level 0.0)))))))
    (:else)))

(fn CanCoverInMud
  (params (CoverAction :type EMudCoverAction%20Enum :default "NewEnumerator0"))
  (return (Math|Boolean|ANDBoolean (not (GameplayTags|HasAnyMatchingGameplayTags (Variables|Default|GetPawn) (Variables|Mud|GetMudCoverBlockedTags))) (ScalableFloat|GetValueasBool (Variables|Mud|Hotfix|GetHF_MudCoverEnabledGlobal)) (ScalableFloat|GetValueasBool (Variables|Mud|Hotfix|GetHF_MudCoverEnabledForClass)) (CallFunction|GetCoverActionIsEnabled CoverAction))))

(fn CheckForApplySlidingParams
  (bind _pawn (Variables|Default|GetPawn)
    (:then
      (if (and (or (Class|MudFunctionLibrary|IsCoveredinMud _pawn) (Class|MudFunctionLibrary|IsonMud _pawn)) (not (Variables|Default|GetAppliedSlidingParams)))
        (CallFunction|ApplyMudSlidingParameters)))
    (:else)))

(fn CheckForResetSlidingParams
  (bind _pawn (Variables|Default|GetPawn)
    (:then
      (if (and (not (or (Class|MudFunctionLibrary|IsCoveredinMud _pawn) (Class|MudFunctionLibrary|IsonMud _pawn))) (Variables|Default|GetAppliedSlidingParams))
        (CallFunction|ResetSlidingParams)))
    (:else)))

(fn DebugSlidingParams
  (bind _movementcomp (CallFunction|GetCharacterMovementComp))
  (bind _slidingcontrolparams (Variables|/Script/FortniteGame.FortMovementComp_Character|GetSlidingControlParams _movementcomp))
  (bind (_minanalogspeed _brakingdeceleration _walkablefloorangle _slidingisenabled _minspeedtostartsliding _uphillslidestartboost _levelslidestartboost _downhillslidestartboost _maxslidingslopeangle _boostuphillgroundfrictionmultiplier _boostlevelgroundfrictionmultiplier _uphillgroundfrictionmultiplier _levelgroundfrictionmultiplier) (Utilities|Struct|BreakSlidingControlParams _slidingcontrolparams))
  (bind (_value _curve) (Utilities|Struct|BreakScalableFloat _levelgroundfrictionmultiplier))
  (bind (_curvetable _rowname) (Utilities|Struct|BreakCurveTableRowHandle _curve))
  (Utilities|IsValid _movementcomp
    (:Is%20Valid
      (Development|PrintString (Utilities|String|ToString(Name) _rowname))
      (Development|PrintString (Utilities|String|Append "Level Ground Friction Mult: " (Utilities|String|ToString(Float) (ScalableFloat|GetValueatLevel _levelgroundfrictionmultiplier)))))))

(fn EndMudCover
  (Class|MudFunctionLibrary|RemoveMudCover (Variables|Default|GetPawn)))

(fn EndWarningTimer
  (Utilities|Time|ClearandInvalidateTimerbyHandle (Variables|Default|GetWarningTimerHandle)))

(fn GetCharacterMovementComp
  (bind _pawn (Variables|Default|GetPawn)
    (:then
      (bind _asfort_movement_comp_character (Utilities|Casting|CastToFortMovementComp_Character (Variables|/Script/Engine.Character|GetCharacterMovement _pawn)))
      (return _asfort_movement_comp_character))
    (:else)))

(fn GetCoverActionIsEnabled
  (params (CoverAction :type EMudCoverAction%20Enum :default "NewEnumerator0"))
  (switch Utilities|FlowControl|Switch|SwitchonEMudCoverAction CoverAction
    (:NewEnumerator0
      (LocalVariables|SetOutIsEnabled true))
    (:NewEnumerator1
      (LocalVariables|SetOutIsEnabled (ScalableFloat|GetValueasBool (Variables|Mud|Hotfix|GetHF_MudCoverSlidingEnabled))))
    (:NewEnumerator2
      (LocalVariables|SetOutIsEnabled (ScalableFloat|GetValueasBool (Variables|Mud|Hotfix|GetHF_MudCoverSprintingEnabled))))
    (:NewEnumerator3
      (LocalVariables|SetOutIsEnabled (ScalableFloat|GetValueasBool (Variables|Mud|Hotfix|GetHF_MudCoverJumpingEnabled)))))
  (return (LocalVariables|GetOutIsEnabled)))

(fn IsSliding)

(fn IsSlidingModifierEnabled
  (return (ScalableFloat|GetValueasBool (Variables|Default|GetRow_SlidingModifierEnabled))))

(fn ManageRemovalWarningCueWithFlag
  (params (Start :type Boolean))
  ;[1 1 1] Setting this flag so the server has an approximate idea of whether or not the gameplay cue is active on the owning client. Not a perfect analog as we have no way of being notified when the gameplay cue ends naturally (without a tag listener + RPC back to server) but it should cut down significantly on how often we RPC to client
  (bind _output_get (Variables|Default|SetRemovalWarningCueIsActive Start))
  (if _output_get
    (CallFunction|OwningClientStartRemovalWarningCue)
    (else
      (CallFunction|OwningClientTryEndRemovalWarningCue))))

(fn OnMudCoverEnd
  (Variables|Default|SetRemovalWarningCueIsActive false)
  (CallFunction|EndWarningTimer))

(fn OnRemovalWarningTimerEnded
  (CallFunction|ManageRemovalWarningCuewithFlag true))

(fn OnStartedSlidingInMud
  (CallFunction|TryCoverinMud "NewEnumerator1"))

(fn RemoveOnMudGameplayEffects
  (bind _pawn (Variables|Default|GetPawn)
    (:then
      (GameplayEffects|RemoveActiveEffectswithTags (Ability|GetAbilitySystemComponent _pawn) (Variables|Mud|GetOnMudEffectAssetTags))
      (Class|MudFunctionLibrary|TryRemoveSharedMudEffects (Variables|Default|GetPawn)))
    (:else)))

(fn ResetSlidingParams
  (if (not (Variables|Mud|Transient|GetbIsSliding))
    (Variables|Default|SetAppliedSlidingParams false))
  (Variables|/Script/FortniteGame.FortMovementComp_Character|SetSlidingControlParams (Variables|Mud|Transient|GetDefaultSlidingControlParams) (CallFunction|GetCharacterMovementComp)))

(fn SetMudCustomDepth
  (params (Enabled :type Boolean))
  (Rendering|SetCustomDepthStencilValue (Variables|/Script/Engine.Character|GetMesh (Variables|Default|GetPawn)) (Math|Integer|SelectInt (Variables|Default|GetMudCustomDepth) (Variables|Default|GetDefaultCustomDepth) Enabled)))

(fn StartCoverInMud
  (params (CoverAction :type EMudCoverAction%20Enum :default "NewEnumerator0"))
  (bind _pawn (Variables|Default|GetPawn)
    (:then
      (LocalVariables|SetOwningASC (Ability|GetAbilitySystemComponent _pawn))
      (Utilities|SwitchHasAuthority
        (:Authority
          ;[1 1 1] Mud covered effects/abilities are simply refreshed when sliding in mud while already covered, so we execute the cue here so that it plays on every refresh
          (GameplayCue|ExecuteGameplayCue (Variables|Default|GetPawn) (Variables|Mud|GetCue_MudCover_Apply_Tag))
          ;[1 1 1] Applies or refreshes the duration of the mud cover GE. This handles logic specific to the mud covered state - see GE_Mud_Covered and GA_Mud_Covered
          (bind _returnvalue (GameplayEffects|ApplyGameplayEffectToSelf :self (LocalVariables|GetOwningASC) :GameplayEffectClass (Variables|Mud|GetGE_CoveredInMud) :Level 0.0))
          (Variables|Default|SetCoveredActiveEffect _returnvalue)
          (CallFunction|UpdateRemovalWarningTimer))))
    (:else)))

(fn StartRemovalWarningCue
  (Projectile|AddGameplayCue%28Looping%29Local (Variables|Default|GetPawn) (Variables|Default|GetRemovalWarningCue) (Ability|GameplayCue|MakeGameplayCueParameters 0.0)))

(fn TryEndRemovalWarningCue
  ;[1 1 1] Removes the looping removal warning cue if it was active. This can be called by UpdateRemovalWarningTimer, or by GA_Mud_Covered when the ability ends as the cue should end before the actual removal sounds play
  (Projectile|RemoveGameplayCue%28Looping%29Local (Variables|Default|GetPawn) (Variables|Default|GetRemovalWarningCue) (Ability|GameplayCue|MakeGameplayCueParameters 0.0)))

(fn TryCoverInMud
  (params (CoverAction :type EMudCoverAction%20Enum :default "NewEnumerator0"))
  (if (CallFunction|CanCoverinMud CoverAction)
    (CallFunction|StartCoverinMud CoverAction)))

(fn UpdateRemovalWarningTimer
  (bind _hf_warningtime (Variables|Default|GetHF_WarningTime))
  (bind _returnvalue (ScalableFloat|GetValueatLevel _hf_warningtime))
  (if (ScalableFloat|GetValueasBool (Variables|Default|GetHF_WarningCueEnabled))
    ;[1 1 1] There's a chance that updating the duration (from fire, for example) already removed mud instantly, and in that case we don't need to start the removal warning timer
    (if (Class|MudFunctionLibrary|IsCoveredinMud (Variables|Default|GetPawn))
      (bind _returnvalue_2 (Ability|GameplayEffect|GetActiveGameplayEffectRemainingDuration (Variables|Default|GetCoveredActiveEffect)))
      (if (<= (- _returnvalue_2 _returnvalue) 0)
        ;[1 1 1] If the remaining duration is already less than the warning time, immediately start the warning cue if it isn't already active, and end the timer in case we were counting down
        (if (not (Variables|Default|GetRemovalWarningCueIsActive))
          (CallFunction|ManageRemovalWarningCuewithFlag true))
        (CallFunction|EndWarningTimer)
        (else
          ;[1 1 1] If the remaining duration is greater than the warning time, set the warning timer to the remaining duration - warning time
          (bind _returnvalue_3 (Utilities|Time|SetTimerbyEvent (EventDispatchers|CreateEvent :!event OnRemovalWarningTimerEnded) (- _returnvalue_2 _returnvalue)))
          (Variables|Default|SetWarningTimerHandle _returnvalue_3)
          (if (Variables|Default|GetRemovalWarningCueIsActive)
            (CallFunction|ManageRemovalWarningCuewithFlag false))))
      (else
        ;[1 1 1] End timer in case we were already counting down and end cue in case it ws already playing
        (CallFunction|EndWarningTimer)
        (if (Variables|Default|GetRemovalWarningCueIsActive)
          (CallFunction|ManageRemovalWarningCuewithFlag false))))))

(fn WarningCueIsActive
  (return (GameplayTags|HasMatchingGameplayTag (Variables|Default|GetPawn) (Variables|Default|GetRemovalWarningCue))))
