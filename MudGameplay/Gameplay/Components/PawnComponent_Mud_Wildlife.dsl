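; Wildlife pawn component for the mud system: covers creatures in mud when they sprint on mud and ends cover when they gain a swimming tag.
; Sets coverage as custom primitive data on the body mesh, rejecting alpha changes whose direction contradicts the covered state.
; Covering also sends an event to attached player pawns.
; A delayed rep-notify failsafe and forced cue-tag removal guard against stuck mud visuals.
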(event EventBeginPlay
  (|CallParentFunction)
  ;[1 1 1] Check for wildlife start swimming as a rough analog for entering water; ideally would be more precise but I think this is all we have at the moment
  (bind _asynctaskproxy (Ability|Tasks|WaitGameplayTagAddToActor (Variables|Default|GetPawn) (Variables|Default|GetSwimmingTag))
    (:then
      (Variables|Default|SetSwimmingTagListenerAction _asynctaskproxy))
    (:Added
      (Development|PrintString "Wildlife in water")
      (CallFunction|EndMudCover))))

(event EventEndPlay
  (params (EndPlayReason :type EEndPlayReason%20Enum :default "Destroyed"))
  (|CallParentFunction EndPlayReason)
  (bind _swimmingtaglisteneraction (Variables|Default|GetSwimmingTagListenerAction)
    (:then
      (Ability|Async|EndAction _swimmingtaglisteneraction))
    (:else)))

(event EventOnCreatureSprintingTagCountChanged
  (params (Tag :type Gameplay%20Tag%20Structure) (NewCount :type Integer))
  ;[1 1 1] Sprinting started
  (if (> NewCount 0)
    (CallFunction|OnStartedSprinting)))

(event EventOnEnteredMud
  (|CallParentFunction)
  ;[1 1 1] Try to cover us in mud if we're already sprinting when we step on to mud
  (if (CallFunction|IsSprinting)
    (CallFunction|TryCoverinMud "NewEnumerator2")))

(event Custom|MudVisualsFailsafe
  (bind _pawn (Variables|Default|GetPawn))
  (bind _returnvalue (Class|MudFunctionLibrary|IsCoveredinMud _pawn))
  ;[1 1 1] Failsafe to ensure we never permanently end up in the incorrect visual state. Can cause ugly pops but it's better than getting stuck
  ; (Triggered by OnRep_bIsOnMudVisualFailsafe)
  (CallFunction|SetMudCoveragePrimitiveData (select _returnvalue (Variables|Default|GetMaxMudCoverage) 0.0))
  (Variables|Default|SetLastMudCoverageAlpha (select _returnvalue 1.0 0.0))
  (if (not _returnvalue)
    (Projectile|RemoveGameplayCue%28Looping%29Local (Variables|Default|GetPawn) (Variables|Default|GetMudCoveredCue) (Ability|GameplayCue|MakeGameplayCueParameters 0.0))))

(fn CanCoverInMud
  (params (CoverAction :type EMudCoverAction%20Enum :default "NewEnumerator0"))
  (return (|CallParentFunction CoverAction)))

(fn EndMudCover
  (bind _mudcoveredcue (Variables|Default|GetMudCoveredCue))
  (|CallParentFunction)
  ;[1 1 1] If wildlife gets covered in mud while not ridden/locally controlled (e.g. by double jumping above mud so that when the animal lands it is no longer ridden), then there are issues with the mud covered gameplay cue being properly removed if it becomes ridden again before the mud is removed, by any means. In this state, it is unclear if the actual gameplay cue is not being removed or just its tag, but either way the only solution that seems to work is to forcibly remove the gameplay cue tag
  (if (GameplayTags|HasMatchingGameplayTag (Variables|Default|GetPawn) _mudcoveredcue)
    (Ability|GameplayEffect|RemoveLooseGameplayTags (Variables|Default|GetPawn) (GameplayTags|MakeGameplayTagContainerfromTag _mudcoveredcue))))

(fn GetSkeletalMesh
  (bind _skelmesh (Variables|Default|GetSkelMesh)
    (:then
      (return _skelmesh))
    (:else
      (bind _returnvalue (Customization|GetSkeletalMeshforPartType (Actor|GetComponentbyClass (Components|GetOwner) "/Script/FortniteGame.FortCharacterPartsComponent") "Body" :!class /Script/FortniteGame.FortCharacterPartsComponent))
      (Utilities|IsValid _returnvalue
        (:Is%20Valid
          (bind _output_get (Variables|Default|SetSkelMesh _returnvalue))
          (return _output_get))))))

(fn IsSprinting
  (bind _returnvalue (GameplayTags|HasMatchingGameplayTag(Message) (Variables|Default|GetPawn) (Variables|Tags|Wildlife|GetCreatureSprintingTag)))
  (return _returnvalue))

(fn OnRep_bIsOnMudVisualFailsafe
  (Utilities|Time|SetTimerbyEvent (EventDispatchers|CreateEvent :!event MudVisualsFailsafe) (ScalableFloat|GetValueatLevel (Variables|Default|GetRow_VisualsFailsafeDelay))))

(fn OnStartedSprinting
  ;[1 1 1] Try to cover us in mud if we start sprinting while already on a mud surface
  (if (Variables|Mud|Transient|GetbIsInMud)
    (CallFunction|TryCoverinMud "NewEnumerator0")))

(fn SetMudCoverageAmount
  (params (Alpha :type Float%20%28double-precision%29) (Increasing :type Boolean))
  ;[1 1 1] Verifies that the mud coverage amount is being set in the right direction before actually setting it, and calls back into the gameplay cue if it's going the wrong direction to stop the timeline the alpha came from
  (if (== (Class|MudFunctionLibrary|IsCoveredinMud (Components|GetOwner)) Increasing)
    (CallFunction|SetMudCoveragePrimitiveData (* (Variables|GetAlpha) (Variables|Default|GetMaxMudCoverage)))
    (Variables|Default|SetLastMudCoverageAlpha (Variables|GetAlpha))
    (return false)
    (else
      (return true))))

(fn SetMudCoveragePrimitiveData
  (params (Value :type Float%20%28single-precision%29))
  (bind _mesh (CallFunction|GetSkeletalMesh))
  (Utilities|IsValid _mesh
    (:Is%20Valid
      (Rendering|Material|SetCustomPrimitiveDataFloat _mesh (Variables|Default|GetMudCoverDataIndex) Value))))

(fn StartCoverInMud
  (params (CoverAction :type EMudCoverAction%20Enum :default "NewEnumerator0"))
  (bind _pawn (Variables|Default|GetPawn))
  (|CallParentFunction CoverAction)
  (for _array_element (Actor|GetAttachedActors _pawn true false)
    (bind _asfort_player_pawn (Utilities|Casting|CastToFortPlayerPawn _array_element))
    (Ability|SendGameplayEventtoActor _asfort_player_pawn (Variables|Tags|GetAttachParentCoveredEventTag) (Utilities|Struct|MakeGameplayEventData "(TagName=\"\")" _pawn))))
