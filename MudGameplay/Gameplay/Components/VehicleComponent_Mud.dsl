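; Vehicle component that covers a Fortnite Athena vehicle in mud on mud surfaces and clears it on leaving them or entering water.
; Surface changes, ignored while airborne, apply or remove a gameplay effect, with forced removal of lingering tags.
; Mud tag changes drive the server's looping gameplay cue and a replicated flag whose delayed failsafe corrects stuck visuals.
; Mud coverage is written to the skeletal mesh as custom primitive data.
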(event EventBeginPlay
  (bind _vehicleowner (Variables|Default|GetVehicleOwner))
  (bind _asfort_athena_vehicle (Utilities|Casting|CastToFortAthenaVehicle (Components|GetOwner))
    (:then
      (bind _output_get (Variables|Default|SetVehicleOwner _asfort_athena_vehicle))
      (Utilities|FlowControl|Sequence
        (:then_0
          ;[1 1 1] Listen for surface changes to detect when we are on mud
          (SurfaceType|BindEventtoOnSurfaceTypeVehicleIsonChangedDelegate _output_get (EventDispatchers|CreateEvent :!event OnSurfaceChanged)))
        (:then_1
          ;[1 1 1] Listen for mud covered tag being added and removed to update visual state and behavior
          (Ability|Tasks|WaitGameplayTagAddToActor _vehicleowner (Variables|Default|GetMudCoveredTag)
            (:then
              (Ability|Tasks|WaitGameplayTagRemoveFromActor _vehicleowner (Variables|Default|GetMudCoveredTag)
                (:Removed
                  (CallFunction|OnEndCoveredinMud))))
            (:Added
              (CallFunction|OnStartCoveredinMud))))
        (:then_2
          ;[1 1 1] Listen for entering water to remove mud coverage
          (Vehicle|BindEventtoOnVehicleEnteredWaterBody (Variables|Default|GetVehicleOwner) (EventDispatchers|CreateEvent :!event OnVehicleEnteredWater)))))
    (:CastFailed
      (Development|PrintString "Mud Vehicle Component could not find a vehicle owner. It may be attached to the wrong kind of actor" true true "(R=1.000000,G=0.000000,B=0.089728,A=1.000000)" 1e+01))))

(event Custom|MudVisualsFailsafe
  (bind _bisonmud (Variables|Default|GetbIsOnMud))
  ;[1 1 1] Failsafe to make sure we never permanently end up in the incorrect visual state. Will likely cause ugly pops but better than getting stuck
  ; (Triggered by OnRep_bIsOnMudVisualFailsafe)
  (Development|PrintString (Utilities|String|Append "Failsafe - on mud: " (Utilities|String|ToString(Boolean) _bisonmud)))
  (CallFunction|SetMudPrimitiveData (select _bisonmud (Variables|Default|GetMaxMudCoverage) 0.0))
  (Variables|Default|SetLastMudCoverageAlpha (select _bisonmud 1.0 0.0))
  (if _bisonmud
    (if (CallFunction|HasMudCoveredCue)
      (Projectile|AddGameplayCue%28Looping%29Local (Variables|Default|GetVehicleOwner) (Variables|Default|GetMudCoveredCue) (Ability|GameplayCue|MakeGameplayCueParameters 0.0)))
    (elif (CallFunction|HasMudCoveredCue)
      (Projectile|RemoveGameplayCue%28Looping%29Local (Variables|Default|GetVehicleOwner) (Variables|Default|GetMudCoveredCue) (Ability|GameplayCue|MakeGameplayCueParameters 0.0)))))

(fn CoverInMud
  (Utilities|SwitchHasAuthority
    (:Authority
      (bind _vehicleowner (Variables|Default|GetVehicleOwner)
        (:then
          (bind _returnvalue (GameplayEffects|ApplyGameplayEffectToSelf :self (Vehicle|GetVehicleAbilitySystemComponent _vehicleowner) :GameplayEffectClass (Variables|Default|GetMudCoveredGE) :Level 0.0))
          (Variables|Default|SetMudCoveredHandle _returnvalue)
          (GameplayCue|ExecuteGameplayCue (Variables|Default|GetVehicleOwner) (Variables|Default|GetMudSplashCue)))
        (:else)))))

(fn GetSkeletalMesh
  (return (Variables|/Script/FortniteGame.FortAthenaVehicle|GetSkeletalMesh (Variables|Default|GetVehicleOwner))))

(fn HasMudCoveredCue
  (return (GameplayTags|HasMatchingGameplayTag (Variables|Default|GetVehicleOwner) (Variables|Default|GetMudCoveredCue))))

(fn OnEndCoveredInMud
  (Development|PrintString "Vehicle END covered in mud")
  (if (Networking|IsServer)
    (GameplayCue|RemoveGameplayCue (Variables|Default|GetVehicleOwner) (Variables|Default|GetMudCoveredCue))
    (Variables|Default|SetbIsOnMudVisualFailsafe false)))

(fn OnEnteredMud
  (Development|PrintString "Vehicle ENTERED mud")
  (CallFunction|UpdateMudState))

(fn OnExitedMud
  (Development|PrintString "Vehicle EXITED mud")
  (CallFunction|UpdateMudState))

(fn OnRep_bIsOnMudVisualFailsafe
  (Utilities|Time|SetTimerbyEvent (EventDispatchers|CreateEvent :!event MudVisualsFailsafe) (ScalableFloat|GetValueatLevel (Variables|Default|GetRow_VisualsFailsafeDelay))))

(fn OnStartCoveredInMud
  (Development|PrintString "Vehicle START covered in mud")
  (if (Networking|IsServer)
    (GameplayCue|AddGameplayCue (Variables|Default|GetVehicleOwner) (Variables|Default|GetMudCoveredCue))
    (Variables|Default|SetbIsOnMudVisualFailsafe true)))

(fn OnSurfaceChanged
  (params (SurfaceTypeVehicleIsOn :type EPhysicalSurface%20Enum :default "SurfaceType_Default"))
  (bind _returnvalue (Utilities|Enum|Equal(Enum) SurfaceTypeVehicleIsOn "SurfaceType30"))
  ;[1 1 1] Surface types seem to change when a vehicle gets air, which can cause rapid switching between being on and off mud, so disregard surface changes while the vehicle is in the air
  (if (not (Vehicle|VehicleIsinAir (Variables|Default|GetVehicleOwner)))
    (Variables|Default|SetbIsOnMud _returnvalue)
    (Utilities|FlowControl|Sequence
      (:then_0
        (if _returnvalue
          (if (not (Utilities|Enum|Equal(Enum) (Variables|Default|GetPrevSurface) "SurfaceType30"))
            (CallFunction|OnEnteredMud))
          (elif (Utilities|Enum|Equal(Enum) (Variables|Default|GetPrevSurface) "SurfaceType30")
            (CallFunction|OnExitedMud))))
      (:then_1
        (Variables|Default|SetPrevSurface SurfaceTypeVehicleIsOn)))))

(fn OnVehicleEnteredWater
  (params (Vehicle :type Fort%20Athena%20Vehicle%20Object%20Reference) (WaterBody :type Fort%20Water%20Body%20Actor%20Object%20Reference))
  (Class|MudFunctionLibrary|RemoveMudCover Vehicle))

(fn RemoveMud
  (bind _vehicleowner (Variables|Default|GetVehicleOwner))
  (GameplayEffects|RemoveActiveGameplayEffect (Vehicle|GetVehicleAbilitySystemComponent _vehicleowner) (Variables|Default|GetMudCoveredHandle) -1)
  ;[1 1 1] If the vehicle had mud applied while not driven by a player (e.g. it rolls into mud after the player hops out), but then removes mud after the player has re-entered, for whatever reason removing the gameplay effect on the server can fail to remove the tag from the owning client, which in turn causes the gameplay cue to not remove properly. So we check if the tag persisted and forcibly remove it if so
  (if (Class|MudFunctionLibrary|IsCoveredinMud _vehicleowner)
    (Ability|GameplayEffect|RemoveLooseGameplayTags _vehicleowner (GameplayTags|MakeGameplayTagContainerfromTag (Variables|Default|GetMudCoveredTag)))))

(fn SetMudCoverageAlpha
  (params (InAlpha :type Float%20%28double-precision%29) (Increasing :type Boolean))
  (bind _vehicleowner (Variables|Default|GetVehicleOwner))
  (bind _returnvalue (Class|MudFunctionLibrary|IsCoveredinMud _vehicleowner))
  (bind _inalpha (Variables|GetInAlpha))
  (if (== _returnvalue Increasing)
    (Development|PrintString (Utilities|String|ToString(Float) _inalpha))
    (CallFunction|SetMudPrimitiveData (* _inalpha (Variables|Default|GetMaxMudCoverage)))
    (Variables|Default|SetLastMudCoverageAlpha (Variables|GetInAlpha))
    (return false)
    (else
      (Development|PrintString (Utilities|String|ToString(Text) (Utilities|Text|FormatText "NSLOCTEXT(\"[DE1E3F8D700DE76334573A5D27A48898]\", \"D99EDA87482D8B6F30A67AA42C59D92F\", \"Mud coverage was {0} but direction was {1}\")" _returnvalue Increasing)))
      (Development|PrintString "Coverage direction does not match mud state")
      (return false))))

(fn SetMudPrimitiveData
  (params (Value :type Float%20%28single-precision%29))
  (bind _skeletal_mesh (CallFunction|GetSkeletalMesh))
  (Utilities|IsValid _skeletal_mesh
    (:Is%20Valid
      (Rendering|Material|SetCustomPrimitiveDataFloat _skeletal_mesh (Variables|Default|GetMudCoverDataIndex) Value))))

(fn UpdateMudState
  (if (Variables|Default|GetbIsOnMud)
    (CallFunction|CoverinMud)
    (else
      (CallFunction|RemoveMud))))
